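# Niagara system for BerryCard ring effects in WildEstate, judging by its name and folder.
# It holds a single emitter named SoundWaves, with no renderers listed.
# SystemState is the only module listed, and the system exposes no user variables, data interfaces or dynamic inputs.
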NiagaraSystem /WildEstate/Assets/FX/BerryCard/NS_WildEstate_BerryCard_Rings
dependencies {usedDataInterfaces [], usedDynamicInputs [], usedModules [&SystemState]}
emitters {SoundWaves {rendererClasses [], renderers []}}
summary {systemName=NS_WildEstate_BerryCard_Rings, emitters [{emitterName=SoundWaves, rendererClasses []}], userVariables []}
